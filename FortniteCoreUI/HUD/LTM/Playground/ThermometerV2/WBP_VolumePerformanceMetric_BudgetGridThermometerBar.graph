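# Creative memory thermometer HUD widget showing budget usage, optional cell coordinates and a default-hidden warning for outdated values.
# UpdatePerformanceMetrics scales usage to the curve-table MemoryLimit and adds a preview delta; UpdatePerformanceMetricsVK shows bytes as megabytes with units.
# Both drive the child BudgetGridThermometerBar's value, margin and highest-cell pip.
# When the highest cell, including its preview delta, reaches max plus margin, the title loops a red-to-white flash.
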WidgetBlueprint /FortniteCoreUI/HUD/LTM/Playground/ThermometerV2/WBP_VolumePerformanceMetric_BudgetGridThermometerBar
animations [
 Max {
  binding_count=1
  length_seconds=0.75
  object=&self:Max
  sequence {bindings [Title {
     id=#1
     possessed_class=CommonTextBlock
     tracks [{
       type=MovieSceneColorTrack
       display_name=Color and Opacity
       sections [{
         type=MovieSceneColorSection
         end_frame=15
         start_frame=0
         channels [
          Color.R {type=MovieSceneScriptingFloatChannel, keys [[0,0.6939],[7,1],[15,1]]}
          Color.G {type=MovieSceneScriptingFloatChannel, keys [[0,0],[7,1],[15,1]]}
          Color.B {type=MovieSceneScriptingFloatChannel, keys [[0,0.0123],[7,1],[15,1]]}
          Color.A {type=MovieSceneScriptingFloatChannel, const [1,1]}
         ]
        }]
      }]
    }]}
 }
 NotMax {
  binding_count=1
  length_seconds=5
  object=&self:NotMax
  sequence {bindings [Title {
     id=#2
     possessed_class=CommonTextBlock
     tracks [{
       type=MovieSceneColorTrack
       display_name=Color and Opacity
       sections [{
         type=MovieSceneColorSection
         end_frame=0
         start_frame=0
         channels [
          Color.R {type=MovieSceneScriptingFloatChannel, const [1,1]}
          Color.G {type=MovieSceneScriptingFloatChannel, const [1,1]}
          Color.B {type=MovieSceneScriptingFloatChannel, const [1,1]}
          Color.A {type=MovieSceneScriptingFloatChannel, const [1,1]}
         ]
        }]
      }]
    }]}
 }
]
graphs [
 EventGraph {nodes [
   Event Construct()
   > SetMaxValue!(
    execute
    self:Budget Grid Thermometer Bar Object Reference=Get BudgetGridThermometerBar() -> (BudgetGridThermometerBar:Budget Grid Thermometer Bar Object Reference)
    Value:double=GetMemoryLimitForThermometerBar!() -> (ThermometerBarVaxValue:double)
   ) -> (then)
   Event PreConstruct()
  ]}
 GetBytesAsMegaBytes {nodes [
   n1 = Inputs!() -> (Bytes:int)
   Outputs!(MegaBytes:double=float / float(A="To Float (Integer)"(InInt=@n1), B=1000000))
  ]}
 GetMemoryLimitForThermometerBar {nodes [
   GetMemoryLimitForThermometerBar() {is=title K2Node_FunctionEntry}
   n1 = Get MemoryLimit() -> (MemoryLimit:Scalable Float Structure)
   > Return Node!(
    execute
    ThermometerBarVaxValue:double="To Float (Integer)"(InInt=GetValueAsInteger!(Input:"Scalable Float Structure (by ref)"=@n1, Level:float=0) -> (ReturnValue:int=0))
   )
  ]}
 GetValueAsPercent {nodes [
   n1 = Inputs!() -> (Value:int, MaxValue:int)
   n2 = "To Float (Integer)"(InInt=@n1)
   n3 = "To Float (Integer)"(InInt=@n1.MaxValue)
   Outputs!(ValuePercent:double=float / float(A="Max (Float)"(A=@n2, B=0), B="Max (Float)"(A=@n3, B=1)) {is=title K2Node_CallFunction})
  ]}
 PlayMaxAnimation {nodes [
   n1 = PlayMaxAnimation!() -> (then, MaxMemory:bool=~) {is=title K2Node_FunctionEntry}
   > n2 = Branch(Condition=@n1)
   > n3 = Branch(Condition=IsAnimationPlaying(InAnimation=Get Max() -> (Max:Widget Animation Object Reference)))
   PlayAnimationTimeRange!(
    execute=@n3.else
    self:User Widget Object Reference
    InAnimation:Widget Animation Object Reference=Get Max()
    StartAtTime:float=0
    EndAtTime:float=0
    NumLoopsToPlay:int=0
    PlayMode:EUMGSequencePlayMode Enum
    PlaybackSpeed:float=1
    bRestoreState:bool=false
   ) -> (then, ReturnValue:Widget Animation Handle Structure)
   Branch(execute=@n2.else, Condition=IsAnimationPlaying(InAnimation=Get Max()))
   > StopAnimation(InAnimation=Get Max())
   > PlayAnimationTimeRange(
    InAnimation=Get NotMax() -> (NotMax:Widget Animation Object Reference)
    StartAtTime=0
    EndAtTime=0
    NumLoopsToPlay=1
    PlaybackSpeed=1
    bRestoreState=false
   )
  ]}
 SetCoordinates {nodes [
   n1 = SetCoordinates!() -> (then, Coordinates_X:int, Coordinates_Y:int) {is=title K2Node_FunctionEntry}
   > Set Text(
    Text=Format Text!(Format="LOCTEXT(\"({X},{Y})\")", X:int=@n1, Y:int=@n1.Coordinates_Y) -> (Result)
    self=Get Coordinates() -> (Coordinates:Common Text Object Reference)
   )
   > SetVisibility(
    self=Get Coordinates()
    InVisibility=Select(Option 0=Collapsed, Option 1=HitTestInvisible, Index=ShowBudgetGridThermometerCoordinates!() -> (ReturnValue:bool))
   )
  ]}
 SetMemoryLimit {nodes [
   n1 = SetMemoryLimit!() -> (then, MemoryLimit:Scalable Float Structure) {is=title K2Node_FunctionEntry}
   > n2 = Set MemoryLimit(MemoryLimit:Scalable Float Structure=@n1) -> (Output_Get:Scalable Float Structure)
   > SetMaxValue(self=Get BudgetGridThermometerBar(), Value="To Float (Integer)"(InInt=GetValueAsInteger(Input=@n2, Level=0) -> (ReturnValue=0)))
  ]}
 SetTitle {nodes [
   n1 = SetTitle!() -> (then, TitleText:Text) {is=title K2Node_FunctionEntry}
   > Set Text(Text=@n1, self=Get Title() -> (Title:Common Text Object Reference))
  ]}
 UpdatePerformanceMetrics {nodes [
   n1 = UpdatePerformanceMetrics!() -> (
    then
    Value:int
    MaxValue:int
    Preview Delta Value:int
    Coordinates:Int Point Structure
    Margin:int
    HighestCellValue:int
    HighestCellPreviewDeltaValue:int
    GlobalMemoryValue:int
   ) {is=title K2Node_FunctionEntry}
   > UpdateThermometerBarRelative!(
    execute
    self
    Value:int=@n1
    MaxValue:int=@n1.MaxValue
    Margin:int=@n1.Margin
    HIghestMemoryValue:int=int + int(A=@n1.HighestCellValue, B=@n1.HighestCellPreviewDeltaValue)
   ) -> (then)
   n2 = Get BudgetGridThermometerBar()
   > n3 = GetValueProgress!(
    execute
    self:Budget Grid Thermometer Bar Object Reference=@n2
    NewValue:int=@n1.Preview Delta Value
    MaxValue:int=@n1.MaxValue
    Margin:int=@n1.Margin
   ) -> (then, ValueProgress:double)
   > SetPreview Value!(execute, self:Budget Grid Thermometer Bar Object Reference=@n2, Preview value:double=@n3) -> (then)
   > SetCoordinates!(execute, self, Coordinates:Int Point Structure=@n1.Coordinates) -> (then) {is=title K2Node_CallFunction}
  ]}
 UpdatePerformanceMetricsVK {nodes [
   n1 = UpdatePerformanceMetricsVK!() -> (then, Value:int, MaxValue:int, Coordinates:Int Point Structure, Margin:int, HighestCellValue:int, HighestCellPreviewDeltaValue:int) {is=title K2Node_FunctionEntry}
   > UpdateThermometerBarAbsolute!(
    execute
    self
    ValueBytes:int=@n1
    MaxValueBytes:int=@n1.MaxValue
    MarginBytes:int=@n1.Margin
    HIghestMemoryValueBytes:int=int + int(A=@n1.HighestCellValue, B=@n1.HighestCellPreviewDeltaValue)
   ) -> (then)
   > SetCoordinates(Coordinates=@n1.Coordinates)
  ]}
 UpdateThermometerBarAbsolute {nodes [
   n1 = UpdateThermometerBarAbsolute!() -> (then, ValueBytes:int, MaxValueBytes:int, MarginBytes:int, HIghestMemoryValueBytes:int) {is=title K2Node_FunctionEntry}
   > Set ValueVarBytes(ValueVarBytes:int=@n1) -> (Output_Get:int=0)
   > Set MaxValueVarBytes(MaxValueVarBytes:int=@n1.MaxValueBytes) -> (Output_Get:int=0)
   > Set MarginVarBytes(MarginVarBytes:int=@n1.MarginBytes) -> (Output_Get:int=0)
   > Set HighestMemoryValueVarBytes(HighestMemoryValueVarBytes:int=@n1.HIghestMemoryValueBytes) -> (Output_Get:int=0)
   > UpdateMarginVisuals!(
    execute
    self:Budget Grid Thermometer Bar Object Reference=Get BudgetGridThermometerBar()
    ExtraValue:int=Get MarginVarBytes() -> (MarginVarBytes:int=0)
   ) -> (then)
   > SetShowUnits!(execute, self:Budget Grid Thermometer Bar Object Reference=Get BudgetGridThermometerBar(), ShowUnits:bool=true) -> (then)
   > SetCurrentValue!(
    execute
    self:Budget Grid Thermometer Bar Object Reference=Get BudgetGridThermometerBar()
    Value:double=Get Bytes as Mega Bytes!(Bytes:int=Get ValueVarBytes() -> (ValueVarBytes:int=0)) -> (MegaBytes:double=~) {is=title K2Node_MacroInstance}
   ) -> (then)
   > SetMaxValue(
    self=Get BudgetGridThermometerBar()
    Value=Get Bytes as Mega Bytes(Bytes=Get MaxValueVarBytes() -> (MaxValueVarBytes:int=0)) -> (MegaBytes=~)
   )
   > SetValueProgress!(
    execute
    self:Budget Grid Thermometer Bar Object Reference=Get BudgetGridThermometerBar()
    NewValue:int=Get ValueVarBytes() -> (ValueVarBytes=0)
    MaxValue:int=Get MaxValueVarBytes() -> (MaxValueVarBytes=0)
    Margin:int=Get MarginVarBytes() -> (MarginVarBytes=0)
   ) -> (then)
   > SetHighestValuePipPosition!(
    execute
    self:Budget Grid Thermometer Bar Object Reference=Get BudgetGridThermometerBar()
    NewValue:int=Get HighestMemoryValueVarBytes() -> (HighestMemoryValueVarBytes:int=0)
    MaxValue:int=Get MaxValueVarBytes() -> (MaxValueVarBytes=0)
    Margin:int=Get MarginVarBytes() -> (MarginVarBytes=0)
   ) -> (then)
   n2 = Get MaxValueVarBytes() -> (MaxValueVarBytes=0)
   n3 = Get MarginVarBytes() -> (MarginVarBytes=0)
   > PlayMaxAnimation!(
    execute
    self
    MaxMemory:bool="integer >= integer"(A=Get HighestMemoryValueVarBytes() -> (HighestMemoryValueVarBytes=0), B=int + int(A=@n2, B=@n3))
   ) -> (then) {is=title K2Node_CallFunction}
  ]}
 UpdateThermometerBarRelative {nodes [
   n1 = UpdateThermometerBarRelative!() -> (then, Value:int, MaxValue:int, Margin:int, HIghestMemoryValue:int) {is=title K2Node_FunctionEntry}
   > Set ValueVar(ValueVar:int=@n1) -> (Output_Get:int=0)
   > Set MaxValueVar(MaxValueVar:int=@n1.MaxValue) -> (Output_Get:int=0)
   > Set MarginVar(MarginVar:int=@n1.Margin) -> (Output_Get:int=0)
   > Set HighestMemoryValueVar(HighestMemoryValueVar:int=@n1.HIghestMemoryValue) -> (Output_Get:int=0)
   > UpdateMarginVisuals(self=Get BudgetGridThermometerBar(), ExtraValue=Get MarginVar() -> (MarginVar:int=0))
   > SetShowUnits(self=Get BudgetGridThermometerBar(), ShowUnits=false)
   n2 = float * float(
    A=Get Value as Percent!(Value:int=Get ValueVar() -> (ValueVar:int=0), MaxValue:int=Get MaxValueVar() -> (MaxValueVar:int=0)) -> (ValuePercent:double=~) {is=title K2Node_MacroInstance}
    B=GetMemoryLimitForThermometerBar() {is=title K2Node_CallFunction}
   ) {is=title K2Node_PromotableOperator}
   > SetCurrentValue(self=Get BudgetGridThermometerBar(), Value="To Float (Integer)"(InInt=Round(A=@n2) {is=title K2Node_CallFunction}))
   > SetValueProgress(
    self=Get BudgetGridThermometerBar()
    NewValue=Get ValueVar() -> (ValueVar=0)
    MaxValue=Get MaxValueVar() -> (MaxValueVar=0)
    Margin=Get MarginVar() -> (MarginVar=0)
   )
   > SetHighestValuePipPosition(
    self=Get BudgetGridThermometerBar()
    NewValue=Get HighestMemoryValueVar() -> (HighestMemoryValueVar:int=0)
    MaxValue=Get MaxValueVar() -> (MaxValueVar=0)
    Margin=Get MarginVar() -> (MarginVar=0)
   )
   n3 = Get MaxValueVar() -> (MaxValueVar=0)
   n4 = Get MarginVar() -> (MarginVar=0)
   > PlayMaxAnimation(MaxMemory="integer >= integer"(A=Get HighestMemoryValueVar() -> (HighestMemoryValueVar=0), B=int + int(A=@n3, B=@n4)))
  ]}
]
variables [
 bShowHighestMemoryCell {type=boolean}
 Coordinates Value {type=IntPoint, default=<0 0>}
 MemoryLimit {
  type=ScalableFloat
  default {
   registryType=None
   value=1
   curve {curveTable=/Game/Athena/Playlists/Creative/AthenaCompositeGD_Creative, rowName=Creative.Thermometer.UI.MemoryLimit}
  }
 }
]
widgets {
 info {
  inheritedWidgetCount=0
  parentClass=/Script/FortniteUI.FortCreativeVolumePerformanceMetricWidget
  rootWidgetClass=&ScaleBox
  widgetCount=10
 }
 widgets [
  ScaleBox_27 {widgetClassPath=&ScaleBox, widget=ScaleBox_27, properties {stretch=UserSpecified, userSpecifiedScale=0.8}}
  Overlay_26 {widgetClassPath=&Overlay, parent=ScaleBox_27, widget=Overlay_26, slot {class=ScaleBoxSlot}}
  VerticalBox_34 {widgetClassPath=&VerticalBox, parent=Overlay_26, widget=VerticalBox_34, slot {class=OverlaySlot}}
  HorizontalBox_0 {
   widgetClassPath=&HorizontalBox
   parent=VerticalBox_34
   widget=HorizontalBox_0
   slot {class=VerticalBoxSlot, properties {horizontalAlignment=HAlign_Center}}
  }
  Title {
   widgetClassPath=&CommonTextBlock
   bIsVariable=true
   parent=HorizontalBox_0
   widget=Title
   properties {
    text=CURRENT MEMORY USAGE
    textTransformPolicy=ToUpper
    font {
     bForceMonospaced=false
     bMaterialIsStencil=false
     fontMaterial=null
     fontObject=&Burbank
     letterSpacing=0
     monospacedWidth=1
     size=20
     skewAmount=0
     typefaceFontName=Burbank-PlayerPower
     outlineSettings {
      bApplyOutlineToDropShadows=true
      bMiteredCorners=false
      bSeparateFillAlpha=false
      outlineMaterial=null
      outlineSize=1
      outlineColor=<rgba 0 0 0 0.75>
     }
    }
    margin=<ltrb 0 -2 2 -1>
    renderTransform {angle=0, scale=<1 1>, shear=<0 0>, translation=<-30 0>}
   }
   slot {class=HorizontalBoxSlot}
  }
  Coordinates {
   widgetClassPath=&CommonTextBlock
   bIsVariable=true
   parent=HorizontalBox_0
   widget=Coordinates
   properties {
    text=(0,0)
    textTransformPolicy=ToUpper
    font {
     bForceMonospaced=false
     bMaterialIsStencil=false
     fontMaterial=null
     fontObject=&Burbank
     letterSpacing=0
     monospacedWidth=1
     size=20
     skewAmount=0
     typefaceFontName=Burbank-PlayerPower
     outlineSettings {
      bApplyOutlineToDropShadows=true
      bMiteredCorners=false
      bSeparateFillAlpha=false
      outlineMaterial=null
      outlineSize=1
      outlineColor=<rgba 0 0 0 0.75>
     }
    }
    margin=<ltrb 0 -2 2 -1>
    renderTransform {angle=0, scale=<1 1>, shear=<0 0>, translation=<-30 0>}
   }
   slot {class=HorizontalBoxSlot, properties {horizontalAlignment=HAlign_Left}}
  }
  BudgetGridThermometerBar {
   widgetClassPath=&/BudgetGridThermometerBar
   bIsVariable=true
   parent=VerticalBox_34
   widget=BudgetGridThermometerBar
   slot {class=VerticalBoxSlot, properties {padding=<ltrb 0 18 0 0>}}
  }
  CookWarningContainer {
   widgetClassPath=&HorizontalBox
   bIsVariable=true
   parent=VerticalBox_34
   widget=CookWarningContainer
   properties {visibility=Collapsed}
   slot {class=VerticalBoxSlot, properties {padding=<ltrb 0 18 0 0>}}
  }
  Image_327 {
   widgetClassPath=&Image
   parent=CookWarningContainer
   widget=Image_327
   properties {brush {
     bIsDynamicallyLoaded=false
     drawAs=Image
     imageType=NoImage
     mirroring=NoMirror
     resourceName=None
     resourceObject=/FortniteCoreUI/HUD/MiniMap/Art/T_UI_Icon_Warning
     tiling=NoTile
     imageSize=<24 24>
     margin=<ltrb 0 0 0 0>
     outlineSettings {
      bUseBrushTransparency=false
      roundingType=HalfHeightRadius
      width=0
      color {colorUseRule=UseColor_Specified, specifiedColor=<rgba 0 0 0 0>}
      cornerRadii=<0 0 0 0>
     }
     tintColor {colorUseRule=UseColor_Specified, specifiedColor=<rgba 1 1 1 1>}
     uVRegion {bIsValid=false, max=<0 0>, min=<0 0>}
    }}
   slot {class=HorizontalBoxSlot, properties {verticalAlignment=VAlign_Center}}
  }
  CookWarningText {
   widgetClassPath=&CommonTextBlock
   parent=CookWarningContainer
   widget=CookWarningText
   properties {
    text=Content changed - launch memory calculation for up-to-date values
    textTransformPolicy=ToUpper
    font {
     bForceMonospaced=false
     bMaterialIsStencil=false
     fontMaterial=null
     fontObject=&Burbank
     letterSpacing=0
     monospacedWidth=1
     size=12.8
     skewAmount=0
     typefaceFontName=Burbank-PlayerPower
     outlineSettings {
      bApplyOutlineToDropShadows=true
      bMiteredCorners=false
      bSeparateFillAlpha=false
      outlineMaterial=null
      outlineSize=1
      outlineColor=<rgba 0 0 0 0.75>
     }
    }
    margin=<ltrb 0 -2 2 -1>
    renderTransform {angle=0, scale=<1 1>, shear=<0 0>, translation=<0 1>}
   }
   slot {class=HorizontalBoxSlot, properties {verticalAlignment=VAlign_Center}}
  }
 ]
}
